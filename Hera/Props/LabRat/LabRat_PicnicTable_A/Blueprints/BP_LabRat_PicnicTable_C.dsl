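; LabRat picnic table prop Blueprint that color-matches its two secondary bench meshes to the building's texture data.
; After the parent construction script runs, it reads the tint at a local index in the building texture data array and applies it to both benches through SetColoronSecondaryMesh.
; An invalid index falls back to neutral grey.
; BeginPlay, overlap and Tick events are empty, so all logic lives in construction.
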(event EventBeginPlay)

(event Collision|EventActorBeginOverlap
  (params (OtherActor :type Actor%20Object%20Reference)))

(event EventTick
  (params (DeltaSeconds :type Float%20%28single-precision%29)))

(fn ConstructionScript
  (bind _sm_labrat_picnicbench_b (Variables|Default|GetSM_LabRat_PicnicBench_B))
  (bind _sm_labrat_picnicbench_b1 (Variables|Default|GetSM_LabRat_PicnicBench_B1))
  (bind _l_tdindex (LocalVariables|GetL_TDIndex))
  (|CallParentFunction)
  (bind _outtexturedataarray (Building|GetBuildingTextureData))
  (if (Utilities|Array|IsValidIndex _outtexturedataarray _l_tdindex)
    (Default|SetColoronSecondaryMesh (Math|Conversions|ToLinearColor(Color) (Variables|/Script/FortniteGame.BuildingTextureData|GetTintColor (Utilities|Array|Get(acopy) _outtexturedataarray _l_tdindex))) _sm_labrat_picnicbench_b)
    (Default|SetColoronSecondaryMesh (Math|Conversions|ToLinearColor(Color) (Variables|/Script/FortniteGame.BuildingTextureData|GetTintColor (Utilities|Array|Get(acopy) _outtexturedataarray _l_tdindex))) _sm_labrat_picnicbench_b1)
    (else
      (Default|SetColoronSecondaryMesh "(R=0.502887,G=0.502887,B=0.502887,A=1.000000)" _sm_labrat_picnicbench_b)
      (Default|SetColoronSecondaryMesh "(R=0.502887,G=0.502887,B=0.502887,A=1.000000)" _sm_labrat_picnicbench_b1))))
